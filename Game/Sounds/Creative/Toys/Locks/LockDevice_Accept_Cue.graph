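# Sound cue for a lock device accept sound in Creative.
# A mixer layers an ItemSpawner item-spawn wave with the LockDevice_Accept_01 wave at slightly reduced input volumes.
# The cue plays pitched up and boosted, using door attenuation and the creative gadgets sound class.
# It references a shared only-one military door concurrency set, with a prevent-new override also defined.
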SoundCue /Game/Sounds/Creative/Toys/Locks/LockDevice_Accept_Cue
root=Mixer_0
nodes [
 {id=Mixer_0, children [WavePlayer_1,WavePlayer_0], properties {inputVolume [0.8,0.7]}}
 {id=WavePlayer_0, properties {soundWaveAssetPtr=&/LockDevice_Accept_01}}
 {id=WavePlayer_1, properties {soundWaveAssetPtr=/Game/Sounds/Creative/Toys/ItemSpawner/ItemSpawner_ItemSpawn}}
]
properties {
 attenuationSettings=/Game/Sounds/Attenuation/Fort_Doors_Att
 maxDistance=2400
 pitchMultiplier=1.3
 priority=0.9
 soundClassObject=&Fort_Gadgets_Creative
 volumeMultiplier=1.5
 concurrencyOverrides {resolutionRule=PreventNew}
 concurrencySet [/Game/Sounds/Creative/Concurrency/MilitaryDoor_OnlyOne_Conc]
}
