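# Sound cue for the hide sound of the Creative lock device toy.
# A single wave player feeds the LockDevice_Hide_01 wave straight to the output at reduced volume.
# It uses the Fort_Doors sound class and door attenuation, and shares the MilitaryDoor_OnlyOne concurrency set.
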SoundCue /Game/Sounds/Creative/Toys/Locks/LockDevice_Hide_Cue
root=WavePlayer_1
nodes [{id=WavePlayer_1, properties {soundWaveAssetPtr=&/LockDevice_Hide_01}}]
properties {
 attenuationSettings=/Game/Sounds/Attenuation/Fort_Doors_Att
 maxDistance=2400
 priority=0.9
 soundClassObject=&Fort_Doors
 volumeMultiplier=0.3
 concurrencyOverrides {resolutionRule=PreventNew}
 concurrencySet [/Game/Sounds/Creative/Concurrency/MilitaryDoor_OnlyOne_Conc]
}
