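# Sound cue for the looping screen sound of a Creative lock device, routed to the Creative gadgets sound class.
# A single looping wave player feeds the output directly at reduced volume.
# Inline attenuation keeps it short-range, falling off from the source with listener focus that lowers off-focus volume.
# Inline concurrency caps it at a few simultaneous instances, stopping the quietest first.
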SoundCue /Game/Sounds/Creative/Toys/Locks/LockDevice_Screen_Loop_Cue
root=WavePlayer_0
nodes [{id=WavePlayer_0, properties {bLooping=true, soundWaveAssetPtr=&/LockDevice_Screen_Loop_02}}]
properties {
 bOverrideAttenuation=true
 bOverrideConcurrency=true
 maxDistance=400
 soundClassObject=&Fort_Gadgets_Creative
 volumeMultiplier=0.5
 attenuationOverrides {
  bEnableFocusInterpolation=true
  bEnableListenerFocus=true
  falloffDistance=400
  nonFocusVolumeAttenuation=0.5
  stereoSpread=500
  attenuationShapeExtents {x=0}
 }
 concurrencyOverrides {maxCount=3, resolutionRule=StopQuietest}
}
