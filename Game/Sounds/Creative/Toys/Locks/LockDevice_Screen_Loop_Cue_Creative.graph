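# Sound cue for the looping screen sound of a Creative lock device, routed to the Creative gadgets sound class.
# A single looping wave player plays the loop wave at reduced volume.
# Overridden attenuation gives a short-range falloff with stereo spread and interpolated listener focus that quiets unfocused instances.
# A concurrency override caps simultaneous instances, stopping the quietest.
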SoundCue /Game/Sounds/Creative/Toys/Locks/LockDevice_Screen_Loop_Cue_Creative
root=WavePlayer_0
nodes [{id=WavePlayer_0, properties {bLooping=true, soundWaveAssetPtr=&/LockDevice_Screen_Loop_02}}]
properties {
 bOverrideAttenuation=true
 bOverrideConcurrency=true
 maxDistance=900
 soundClassObject=&Fort_Gadgets_Creative
 volumeMultiplier=0.6
 attenuationOverrides {
  bEnableFocusInterpolation=true
  bEnableListenerFocus=true
  falloffDistance=800
  nonFocusVolumeAttenuation=0.5
  stereoSpread=1000
  attenuationShapeExtents {x=100}
 }
 concurrencyOverrides {maxCount=3, resolutionRule=StopQuietest}
}
